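# Sound cue for the cast sound of the FloppingRabbit item in Athena, under the gameplay cue sounds.
# A single wave player plays the FloppingRabbit_Cast wave directly as the root, with no randomization or mixing.
# It routes through the gadgets and items sound class with a slight volume boost and a non-binaural gadget activation attenuation asset for spatial falloff.
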SoundCue /FNGameplayCues/Athena/Sounds/Items/FloppingRabbit/FloppingRabbit_Cast_Cue
root=WavePlayer_0
nodes [{id=WavePlayer_0, properties {soundWaveAssetPtr=&/FloppingRabbit_Cast}}]
properties {
 attenuationSettings=/Game/Sounds/Attenuation/Fort_GadgetActivateNonBinaural_Att
 maxDistance=4000
 soundClassObject=&Fort_Gadgets_Items
 volumeMultiplier=1.2
}
