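# Sound cue for the Flopping Rabbit item's player-impact hit, judging by its name and wave.
# A single wave player plays one hit sample through a modulator that randomizes pitch slightly upward for variation.
# It uses a dedicated FloppingRabbit hit attenuation asset and the gadget items sound class, with volume trimmed slightly.
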SoundCue /FNGameplayCues/Athena/Sounds/Items/FloppingRabbit/FloppingRabbit_ImpactPlayer_Cue
root=Modulator_0
nodes [
 {id=Modulator_0, children [WavePlayer_0], properties {pitchMax=1.1, pitchMin=1}}
 {id=WavePlayer_0, properties {soundWaveAssetPtr=&/FloppingRabbit_HitPlayer_01}}
]
properties {
 attenuationSettings=/Game/Sounds/Attenuation/Fort_FloppingRabbitHit_Att
 maxDistance=5300
 soundClassObject=&Fort_Gadgets_Items
 volumeMultiplier=0.9
}
